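# UI-domain material for an item shop tile that builds rarity-colored radial gradient backgrounds.
# Blue, purple, brown and green named reroutes each blend a dark and a light shade through MF_UI_RadialGradient.
# A Rarity vector parameter's channels drive a chain of lerps over rerouted inputs, with zero selecting a color.
# MF_UI_UVScaling publishes fit and fill UV reroutes; reroute targets and material output connections are not shown.
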Material /ItemShop/UI/Art/Rufus/M_UI_Rufus_ShopTile
graph {
 properties {bUsedWithStaticMesh=false, materialDomain=MD_UI, pixelDepthOffsetMode=PDOM_Legacy, usedShadingModels=Default Lit}
 nodes [
  n1 = VectorParameter() {group=Rarity, parameterName="Rarity [set to 0 for color]", defaultValue=<rgba 1 1 1 1>}
  n2 = LinearInterpolate(A=@n1.R, B=NamedRerouteUsage()) {constAlpha=1}
  n3 = NamedRerouteUsage()
  LinearInterpolate(A=NamedRerouteUsage(), B=LinearInterpolate(A=NamedRerouteUsage(), B=LinearInterpolate(A=@n2, B=@n3, Alpha=@n1.G), Alpha=@n1.B), Alpha=@n1.A)
  n4 = NamedRerouteUsage()
  n5 = Constant()
  n6 = Constant() {r=1}
  NamedRerouteDeclaration(None=LinearInterpolate(
    A=Constant3Vector() {constant=<rgba 0.0129 0.0601 0.25 1>}
    B=Constant3Vector() {constant=<rgba 0.0497 0.3564 0.7991 1>}
    Alpha=MaterialFunctionCall(UV=@n4, GradientStart=@n5, GradientEnd=@n6) {materialFunction=&MF_UI_RadialGradient}.Smooth
   )) {name=Blue, nodeColor=<rgba 0 0.377 1 1>}
  n7 = NamedRerouteUsage()
  n8 = Constant()
  n9 = Constant() {r=1}
  NamedRerouteDeclaration(None=LinearInterpolate(
    A=Constant3Vector() {constant=<rgba 0.0782 0.0194 0.2384 1>}
    B=Constant3Vector() {constant=<rgba 0.4342 0.1095 0.7682 1>}
    Alpha=MaterialFunctionCall(UV=@n7, GradientStart=@n8, GradientEnd=@n9) {materialFunction=&MF_UI_RadialGradient}.Smooth
   )) {name=Purple, nodeColor=<rgba 0.7578 0 1 1>}
  n10 = NamedRerouteUsage()
  n11 = Constant()
  n12 = Constant() {r=1}
  NamedRerouteDeclaration(None=LinearInterpolate(
    A=Constant3Vector() {constant=<rgba 0.2705 0.0648 0.0194 1>}
    B=Constant3Vector() {constant=<rgba 0.7084 0.2051 0.0823 1>}
    Alpha=MaterialFunctionCall(UV=@n10, GradientStart=@n11, GradientEnd=@n12) {materialFunction=&MF_UI_RadialGradient}.Smooth
   )) {name=Brown, nodeColor=<rgba 0.3333 0.1476 0 1>}
  n13 = MaterialFunctionCall() {materialFunction=&MF_UI_UVScaling}
  NamedRerouteDeclaration(None=@n13.Scaled to Fit) {name=UVFit, nodeColor=<rgba 0 0.6719 1 1>}
  NamedRerouteDeclaration(None=@n13.Scaled To Fill) {name=UVFill, nodeColor=<rgba 1 0.6562 0 1>}
  n14 = NamedRerouteUsage()
  n15 = Constant()
  n16 = Constant() {r=1}
  NamedRerouteDeclaration(None=LinearInterpolate(
    A=Constant3Vector() {constant=<rgba 0.0097 0.0887 0.0123 1>}
    B=Constant3Vector() {constant=<rgba 0.1195 0.4125 0.0423 1>}
    Alpha=MaterialFunctionCall(UV=@n14, GradientStart=@n15, GradientEnd=@n16) {materialFunction=&MF_UI_RadialGradient}.Smooth
   )) {name=Green, nodeColor=<rgba 0.0547 1 0 1>}
 ]
}
